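# Niagara trail effect for the Arctic Ice Blue glider, drawing thruster exhaust, wing trails and speedlines.
# Mesh flames, fake-light sprites and two yellow ribbon beams attach to skeletal mesh sockets, while SideWings builds two anim trails from socket pairs.
# Timeline and Size scale grow most thruster elements; ForwardDot bends the beams and drives speedline rate and opacity.
# Front-end overrides replace several gameplay-driven values, and time-of-day functions tint the fake lighting.
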NiagaraSystem /BRCosmetics/Gadgets/Assets/VinderTech_GliderChute/Glider_ArcticIceBlue/FX/NS_Glider_ArcticIceBlue_Trail
dependencies {
 usedDataInterfaces [
  &NiagaraDataInterfaceArrayColor
  &NiagaraDataInterfaceVector2DCurve
  &NiagaraDataInterfaceCurve
  &NiagaraDataInterfaceColorCurve
  &NiagaraDataInterfaceVectorCurve
 ]
 usedDynamicInputs [
  &Multiply_VectorByFloat
  &Boolean_AndOperation
  &SetBoolByFloatComparison
 ]
 usedModules [
  &SystemState
  &Timeline
  &self:SystemSpawnScript.SystemScriptSource.SystemScriptGraph.NiagaraNodeAssignment_0.SetVariables_#1
 ]
}
emitters {
 AirThrusters004 {
  emitterUpdateScript [
   &EmitterState()
   &SpawnRate(SpawnRate=16, Use Spawn Probability=false)
  ]
  particleSpawnScript [
   &PartitionParticles()
   &InitializeParticle(Color=<rgba 1 1 1 0.1>, Mesh Scale=<0.1 0.1 0.1>, Mesh Scale Mode=::NewEnumerator4, UsePositionOffset=false, Write Material Random=true)
   &InitialMeshOrientation(
    Mesh Axis To Orient=<0 1 0>
    Mesh Orientation Mode=::NewEnumerator2
    Orientation Vector=<0 0 -1>
    Rotation=&RandomRangeVector(Minimum=<0 0 0>, Maximum=<0 0 1>)
    Use Rotation Vector=true
   )
  ]
  particleUpdateScript [
   &ParticleState()
   &SkeletalMeshLocation(
    Bone / Socket Index=$Particles.PartitionParticles.Partition
    MaskSpawnGroup=false
    Mesh Sampling Type=::NewEnumerator5
    Sampled Position Offset=<0 -2 0>
    Skeletal Mesh=$User.SkeletalMesh
    Socket Sampling Mode=::NewEnumerator5
   )
   &ScaleMeshSize(Scale Factor=&FN_FrontEndVector3Modifier(
     Gameplay=&MakeVector(
      X=1
      Y=1
      Z=&Lerp_Float(
       A=1
       B=3
       Alpha=&ClampFloat(Float=&Add_Float(A=$System.Timeline.NormalizedTime, B=$System.Size scale))
      )
     )
     Front End=<1 1 3>
    )) {module=ScaleMeshSize001}
   &ScaleColor(
    Scale Alpha=&Lerp_Float(A=0, B=0.2, Alpha=$Particles.NormalizedAge)
    Scale RGB=&MakeVectorFromLinearColorRGB(Color=&ColorFromCurve(ColorCurve=DataInterface() {propertyValues {
        AlphaCurve {keys [<at 0 1 RCIM_Cubic>,<at 1 1 RCIM_Cubic>]}
        BlueCurve {keys [<at 0.1329 1 RCIM_Cubic>,<at 0.5656 0.1 RCIM_Cubic>]}
        GreenCurve {keys [<at 0.1329 0.2605 RCIM_Cubic>,<at 0.5656 0.0167 RCIM_Cubic>]}
        RedCurve {keys [<at 0.1329 0 RCIM_Cubic>,<at 0.5656 0 RCIM_Cubic>]}
       }}))
    ScaleA=true
    ScaleRGB=false
   )
   /SetVariables(Particles.ColorAlt:LinearColor=&MakeLinearColorFromVectorAndFloat(
     "Vector (RGB)"=&MakeVectorFromLinearColorRGB(Color=&ColorFromCurve(ColorCurve=DataInterface() {propertyValues {
         AlphaCurve {keys [<at 0 1 RCIM_Cubic>,<at 1 1 RCIM_Cubic>]}
         BlueCurve {keys [
           <at 0.083 1 RCIM_Cubic>
           {arriveTangent=-2.419, interpMode=RCIM_Cubic, time=0.2491, value=0.1}
           {interpMode=RCIM_Cubic, leaveTangent=1.4033, time=0.5429, value=0.0469}
          ]}
         GreenCurve {keys [
           <at 0.083 0 RCIM_Cubic>
           {arriveTangent=0.0614, interpMode=RCIM_Cubic, time=0.2491, value=0.0565}
           {interpMode=RCIM_Cubic, leaveTangent=0.7928, time=0.5429, value=0.0242}
          ]}
         RedCurve {keys [<at 0.083 0.5929 RCIM_Cubic>, {arriveTangent=-1.5047, interpMode=RCIM_Cubic, time=0.2491, value=0}, <at 0.5429 0 RCIM_Cubic>]}
        }}))
     "Float (Alpha)"=&Lerp_Float(A=0, B=0.05, Alpha=$Particles.NormalizedAge)
    )) {isSetParameters=true, module=SetVariables_#2}
  ]
  rendererClasses [&NiagaraMeshRendererProperties]
  renderers [{rendererClass=&NiagaraMeshRendererProperties, rendererIndex=0}]
 }
 AirThrusters005 {
  emitterUpdateScript [
   &EmitterState()
   &SpawnBurst_Instantaneous(Spawn Count=2, Use Spawn Probability=false, UseLoopCountLimit=false)
   &SpawnRate(SpawnRate=2, Use Spawn Probability=false)
  ]
  particleSpawnScript [
   &PartitionParticles()
   &InitializeParticle(
    Color=<rgba 1 0.7772 0 1>
    Lifetime=4
    Mesh Scale=<0.1 0.07 0.07>
    Mesh Scale Mode=::NewEnumerator4
    UsePositionOffset=false
    Write Material Random=true
   )
   &InitialMeshOrientation(
    Mesh Axis To Orient=<0 -1 0>
    Mesh Orientation Mode=::NewEnumerator2
    Rotation=&RandomRangeVector(Minimum=<0 0 0>, Maximum=<1 0 0>)
    Use Rotation Vector=true
   )
  ]
  particleUpdateScript [
   &ParticleState()
   &SkeletalMeshLocation(
    Bone / Socket Index=$Particles.PartitionParticles.Partition
    MaskSpawnGroup=false
    Mesh Sampling Type=::NewEnumerator5
    Sampled Position Offset=<0 2 0>
    Skeletal Mesh=$User.SkeletalMesh
    Socket Sampling Mode=::NewEnumerator5
   )
   &ScaleMeshSize(Scale Factor=&FN_FrontEndVector3Modifier(
     Gameplay=&MakeVector(
      X=&Lerp_Float(
       A=0.1
       B=1
       Alpha=&ClampFloat(Float=&Add_Float(A=$System.Timeline.NormalizedTime, B=$System.Size scale))
      )
      Y=&Lerp_Float(
       A=0.5
       B=1
       Alpha=&ClampFloat(Float=&Add_Float(A=$System.Timeline.NormalizedTime, B=$System.Size scale))
      )
      Z=&Lerp_Float(
       A=0.5
       B=1
       Alpha=&ClampFloat(Float=&Add_Float(A=$System.Timeline.NormalizedTime, B=$System.Size scale))
      )
     )
     Front End=<1 1 1>
    )) {module=ScaleMeshSize001}
   &ScaleColor(
    Scale RGB=&MakeVectorFromLinearColorRGB(Color=&ColorFromCurve(ColorCurve=DataInterface() {propertyValues {
        AlphaCurve {keys [<at 0 1 RCIM_Cubic>,<at 1 1 RCIM_Cubic>]}
        BlueCurve {keys [<at 0.0011 0 RCIM_Cubic>,<at 0.1822 0>,<at 0.5656 0 RCIM_Cubic>]}
        GreenCurve {keys [<at 0.0011 4.5877 RCIM_Cubic>,<at 0.1822 0.9175>,<at 0.5656 0.3586 RCIM_Cubic>]}
        RedCurve {keys [<at 0.0011 5 RCIM_Cubic>,<at 0.1822 1>,<at 0.5656 0.4427 RCIM_Cubic>]}
       }}))
    ScaleA=true
    ScaleRGB=true
   )
   /SetVariables(Particles.ColorAlt:LinearColor=<rgba 0.5552 0.2056 0 1>) {isSetParameters=true, module=SetVariables_#2}
  ]
  rendererClasses [&NiagaraMeshRendererProperties]
  renderers [{rendererClass=&NiagaraMeshRendererProperties, rendererIndex=0}]
 }
 DynamicBeam001 {
  emitterUpdateScript [
   &EmitterState(Life Cycle Mode=::NewEnumerator1, Loop Duration=3, UseLoopDelay=false)
   &SampleSkeletalMesh(
    AdditionalYawPitchRoll=true
    Bone / Socket Index=1
    Mesh Sampling Type:ENiagaraSkelSamplingModeFull=::NewEnumerator5
    Skeletal Mesh:NiagaraDataInterfaceSkeletalMesh=$User.SkeletalMesh
    Socket Sampling Mode:ENiagaraSocketSamplingMode=::NewEnumerator5
   )
   /Niagara/Modules/Beams/BeamEmitterSetup(
    Absolute Beam Start=false
    Beam End:Vector3f=&FN_FrontEndVector3Modifier(
     Gameplay=&Lerp_Vector(
      A=<0 0 0>
      B=<0 120 0>
      Alpha=&ClampFloat(Float=&Add_Float(A=$System.Timeline.NormalizedTime, B=$System.Size scale))
     )
     Front End=<0 120 0>
    )
    Beam Start:Vector3f=$Emitter.SampleSkeletalMesh.SampledPosition
   )
   &SpawnBurst_Instantaneous(Spawn Count=50, Use Spawn Probability=false, UseLoopCountLimit=false)
  ]
  particleSpawnScript [
   &InitializeParticle(Lifetime=3, Position Mode=::NewEnumerator0, Write Material Random=true)
   /Niagara/Modules/Beams/SpawnBeam()
   /Niagara/Modules/Beams/BeamWidth(Beam Width:NiagaraFloat=&FloatFromCurve(
     FloatCurve=DataInterface() {propertyValues {Curve {keys [{tangentMode=RCTM_User, time=0, value=1}, {tangentMode=RCTM_User, time=1, value=0.9}]}}}
     CurveIndex=$Particles.RibbonLinkOrder
     Scale Curve=15.5
    ))
  ]
  particleUpdateScript [
   &ParticleState()
   &Color(
    Color=&ColorFromCurve(
     ColorCurve=DataInterface() {propertyValues {
       AlphaCurve {keys [<at 0 1 RCIM_Cubic>,<at 0.7559 1>,<at 1 0 RCIM_Cubic>]}
       BlueCurve {keys [<at 0 0>,<at 0.422 0>,<at 1 0>]}
       GreenCurve {keys [<at 0 2.6751>,<at 0.422 2.4844>,<at 1 0.4605>]}
       RedCurve {keys [<at 0 5>,<at 0.422 5>,<at 1 0.5927>]}
      }}
     CurveIndex=$Particles.RibbonLinkOrder
    )
    Scale Alpha=0.8
   )
   /Niagara/Modules/Beams/UpdateBeam()
   /SetVariables(Particles.Position=&Add_Vector(
     A=Linked() {linkedVariable {name=Particles.Position, type=NiagaraPosition}}
     B=&MakeVector(
      X=&Multiply_Float(
       A=&Lerp_Float(A=0, B=200, Alpha=$User.ForwardDot)
       B=&FloatFromCurve(
        FloatCurve=DataInterface() {propertyValues {Curve {keys [
            {arriveTangent=-1.0265, interpMode=RCIM_Cubic, leaveTangent=-1.0265, tangentMode=RCTM_User, time=0, value=0}
            {arriveTangent=3.4394, interpMode=RCIM_Cubic, leaveTangent=3.4394, tangentMode=RCTM_User, time=1, value=1}
           ]}}}
        CurveIndex=$StackContext.RibbonLinkOrder
        Scale Curve=&MakeFloatFromVector(
         Vector=&RotateVectorByQuaternion(Vector To Rotate=<1 0 0>, Quaternion=$Emitter.SampleSkeletalMesh.SampledRotation)
         Channel=::NewEnumerator1
        )
       )
      )
      Y=0
      Z=0
     )
    )) {isSetParameters=true, module=SetVariables_#3}
   &SolveForcesAndVelocity(Clamp Velocity=false, Limit Acceleration=false)
   &DynamicMaterialParameters(
    Index 0 Param 1=1
    Index 0 Param 2=1
    Index 0 Param 3=1
    Index 0 Param 4=1
    Index 1 Param 1=1
    Index 1 Param 2=&FloatFromCurve(FloatCurve=DataInterface() {propertyValues {Curve {keys [{tangentMode=RCTM_User, time=0, value=3}, {tangentMode=RCTM_User, time=1, value=0}]}}})
    Index 1 Param 3=&FloatFromCurve(
     FloatCurve=DataInterface() {propertyValues {Curve {keys [
         {interpMode=RCIM_Cubic, tangentMode=RCTM_User, time=0, value=0}
         {arriveTangent=8.9262, interpMode=RCIM_Cubic, leaveTangent=8.9262, tangentMode=RCTM_User, time=1, value=3}
        ]}}}
     CurveIndex=$Particles.RibbonLinkOrder
    )
    Index 1 Param 4=1
    Index 3 Param 1=10
    Index 3 Param 3=1
    Module.Param0WriteEnabled=false
    Module.Param0WriteEnabled1=true
    Module.Param0WriteEnabled2=false
    Module.Param0WriteEnabled3=false
    Module.Param1WriteEnabled=false
    Module.Param1WriteEnabled1=true
    Module.Param1WriteEnabled2=false
    Module.Param1WriteEnabled3=false
    Module.Param2WriteEnabled=false
    Module.Param2WriteEnabled1=true
    Module.Param2WriteEnabled2=false
    Module.Param2WriteEnabled3=true
    Module.Param3WriteEnabled=false
    Module.Param3WriteEnabled1=true
    Module.Param3WriteEnabled2=false
    Module.Param3WriteEnabled3=false
    Write Parameter Index 1=true
    Write Parameter Index 2=true
    Write Parameter Index 3=true
   )
  ]
  rendererClasses [&NiagaraRibbonRendererProperties]
  renderers [{rendererClass=&NiagaraRibbonRendererProperties}]
 }
 DynamicBeam004 {
  emitterUpdateScript [
   &EmitterState(Life Cycle Mode=::NewEnumerator1, Loop Duration=3, UseLoopDelay=false)
   &SampleSkeletalMesh(
    AdditionalYawPitchRoll=true
    Mesh Sampling Type:ENiagaraSkelSamplingModeFull=::NewEnumerator5
    Skeletal Mesh:NiagaraDataInterfaceSkeletalMesh=$User.SkeletalMesh
    Socket Sampling Mode:ENiagaraSocketSamplingMode=::NewEnumerator5
   )
   &BeamEmitterSetup(
    Absolute Beam Start=false
    Beam End:Vector3f=&FN_FrontEndVector3Modifier(
     Gameplay=&Lerp_Vector(
      A=<0 0 0>
      B=<0 120 0>
      Alpha=&ClampFloat(Float=&Add_Float(A=$System.Timeline.NormalizedTime, B=$System.Size scale))
     )
     Front End=<0 120 0>
    )
    Beam Start:Vector3f=$Emitter.SampleSkeletalMesh.SampledPosition
   )
   &SpawnBurst_Instantaneous(Spawn Count=50, Use Spawn Probability=false, UseLoopCountLimit=false)
  ]
  particleSpawnScript [
   &InitializeParticle(Lifetime=3, Position Mode=::NewEnumerator0, Write Material Random=true)
   &SpawnBeam()
   &BeamWidth(Beam Width:NiagaraFloat=&FloatFromCurve(
     FloatCurve=DataInterface() {propertyValues {Curve {keys [{tangentMode=RCTM_User, time=0, value=1}, {tangentMode=RCTM_User, time=1, value=0.9}]}}}
     CurveIndex=$Particles.RibbonLinkOrder
     Scale Curve=15.5
    ))
  ]
  particleUpdateScript [
   &ParticleState()
   &UpdateBeam()
   &Color(
    Color=&ColorFromCurve(
     ColorCurve=DataInterface() {propertyValues {
       AlphaCurve {keys [<at 0 1 RCIM_Cubic>,<at 0.7559 1>,<at 1 0 RCIM_Cubic>]}
       BlueCurve {keys [<at 0 0>,<at 0.422 0>,<at 1 0>]}
       GreenCurve {keys [<at 0 2.6833>,<at 0.422 2.4844>,<at 1 0.4605>]}
       RedCurve {keys [<at 0 5>,<at 0.422 5>,<at 1 0.5927>]}
      }}
     CurveIndex=$Particles.RibbonLinkOrder
    )
    Scale Alpha=0.8
   )
   /SetVariables(Particles.Position=&Add_Vector(
     A=Linked() {linkedVariable {name=Particles.Position, type=NiagaraPosition}}
     B=&MakeVector(
      X=&Multiply_Float(
       A=&Lerp_Float(A=50, B=200, Alpha=$User.ForwardDot)
       B=&FloatFromCurve(
        FloatCurve=DataInterface() {propertyValues {Curve {keys [
            {arriveTangent=-1.0265, interpMode=RCIM_Cubic, leaveTangent=-1.0265, tangentMode=RCTM_User, time=0, value=0}
            {arriveTangent=3.4394, interpMode=RCIM_Cubic, leaveTangent=3.4394, tangentMode=RCTM_User, time=1, value=1}
           ]}}}
        CurveIndex=$StackContext.RibbonLinkOrder
        Scale Curve=&MakeFloatFromVector(
         Vector=&RotateVectorByQuaternion(Vector To Rotate=<1 0 0>, Quaternion=$Emitter.SampleSkeletalMesh.SampledRotation)
         Channel=::NewEnumerator1
        )
       )
      )
      Y=0
      Z=0
     )
    )) {isSetParameters=true, module=SetVariables_#3}
   &SolveForcesAndVelocity(Clamp Velocity=false, Limit Acceleration=false)
   &DynamicMaterialParameters(
    Index 0 Param 1=1
    Index 0 Param 2=1
    Index 0 Param 3=1
    Index 0 Param 4=1
    Index 1 Param 1=1
    Index 1 Param 2=&FloatFromCurve(FloatCurve=DataInterface() {propertyValues {Curve {keys [{tangentMode=RCTM_User, time=0, value=3.5}, {tangentMode=RCTM_User, time=1, value=0.05}]}}})
    Index 1 Param 3=&FloatFromCurve(
     FloatCurve=DataInterface() {propertyValues {Curve {keys [
         {interpMode=RCIM_Cubic, tangentMode=RCTM_User, time=0, value=0}
         {arriveTangent=8.9262, interpMode=RCIM_Cubic, leaveTangent=8.9262, tangentMode=RCTM_User, time=1, value=3}
        ]}}}
     CurveIndex=$Particles.RibbonLinkOrder
    )
    Index 1 Param 4=1
    Index 3 Param 1=10
    Index 3 Param 3=1
    Module.Param0WriteEnabled=false
    Module.Param0WriteEnabled1=true
    Module.Param0WriteEnabled2=false
    Module.Param0WriteEnabled3=false
    Module.Param1WriteEnabled=false
    Module.Param1WriteEnabled1=true
    Module.Param1WriteEnabled2=false
    Module.Param1WriteEnabled3=false
    Module.Param2WriteEnabled=false
    Module.Param2WriteEnabled1=true
    Module.Param2WriteEnabled2=false
    Module.Param2WriteEnabled3=true
    Module.Param3WriteEnabled=false
    Module.Param3WriteEnabled1=true
    Module.Param3WriteEnabled2=false
    Module.Param3WriteEnabled3=false
    Write Parameter Index 1=true
    Write Parameter Index 2=true
    Write Parameter Index 3=true
   )
  ]
  rendererClasses [&NiagaraRibbonRendererProperties]
  renderers [{rendererClass=&NiagaraRibbonRendererProperties}]
 }
 FakeLighting {
  emitterUpdateScript [
   &EmitterState()
   &SpawnBurst_Instantaneous(Spawn Count=4, Use Spawn Probability=false, UseLoopCountLimit=true)
  ]
  particleSpawnScript [
   &PartitionParticles()
   &InitializeParticle(Color=<rgba 1 1 1 0.5>, Sprite Size Mode=::NewEnumerator3, Uniform Sprite Size=55, UsePositionOffset=false, Write Material Random=true)
   &DynamicMaterialParameters(
    Index 0 Param 1=1
    Index 0 Param 2=1
    Index 0 Param 3=1
    Index 0 Param 4=1
    Module.Param0WriteEnabled=true
    Module.Param1WriteEnabled=true
    Module.Param2WriteEnabled=true
    Module.Param3WriteEnabled=true
   )
  ]
  particleUpdateScript [
   &ParticleState(
    Kill Particles When Lifetime Has Elapsed=false
    Let Infinitely Lived Particles Die When Emitter Deactivates=true
    Loop Particles Lifetime=true
   )
   &SkeletalMeshLocation(
    Bone / Socket Index=$Particles.PartitionParticles.Partition
    MaskSpawnGroup=false
    Mesh Sampling Type=::NewEnumerator5
    Skeletal Mesh=$User.SkeletalMesh
    Socket Sampling Mode=::NewEnumerator5
   )
   &CameraOffset()
   &ScaleSpriteSize(
    Scale Sprite Size Mode=::NewEnumerator0
    Uniform Scale Factor=&MakeCustomFloatFromBool(
     Bool=&SetBoolByIntComparison(A=$Particles.PartitionParticles.Partition, B=1, Comparison Type=::NewEnumerator0)
     True Float=0
     False Float=1
    )
   )
   &ScaleColor(
    Scale Alpha=&Lerp_Float(
     A=0.25
     B=&Multiply_Float(
      A=1
      B=&FloatFromCurve(
       FloatCurve=DataInterface() {propertyValues {Curve {keys [
           <at 0 1 RCIM_Cubic>
           {arriveTangent=0.1177, interpMode=RCIM_Cubic, leaveTangent=0.1177, time=0.0746, value=0.8987}
           {arriveTangent=-0.0227, interpMode=RCIM_Cubic, leaveTangent=-0.0227, tangentMode=RCTM_User, time=0.1682, value=1.0198}
           {arriveTangent=-0.7897, interpMode=RCIM_Cubic, leaveTangent=-0.7897, time=0.2019, value=0.7756}
           {arriveTangent=0.8345, interpMode=RCIM_Cubic, leaveTangent=0.8345, time=0.2748, value=0.9356}
           {arriveTangent=0.5127, interpMode=RCIM_Cubic, leaveTangent=0.5127, time=0.3464, value=0.8961}
           {arriveTangent=-0.6889, interpMode=RCIM_Cubic, leaveTangent=-0.6889, time=0.4074, value=1.0036}
           {arriveTangent=-0.8737, interpMode=RCIM_Cubic, leaveTangent=-0.8737, time=0.4365, value=0.834}
           {arriveTangent=-0.4832, interpMode=RCIM_Cubic, leaveTangent=-0.4832, time=0.4871, value=0.9339}
           {arriveTangent=0.1715, interpMode=RCIM_Cubic, leaveTangent=0.1715, time=0.5331, value=0.7873}
           {arriveTangent=0.795, interpMode=RCIM_Cubic, leaveTangent=0.795, time=0.6112, value=0.9552}
           {arriveTangent=0.9015, interpMode=RCIM_Cubic, leaveTangent=0.9015, time=0.69, value=0.912}
           {arriveTangent=-0.4254, interpMode=RCIM_Cubic, leaveTangent=-0.4254, time=0.7536, value=1.0835}
           {arriveTangent=-1.4701, interpMode=RCIM_Cubic, leaveTangent=-1.4701, time=0.7995, value=0.8654}
           {arriveTangent=-0.4912, interpMode=RCIM_Cubic, leaveTangent=-0.4912, time=0.8674, value=0.9161}
           {arriveTangent=0.6326, interpMode=RCIM_Cubic, leaveTangent=0.6326, time=0.9138, value=0.8093}
           {interpMode=RCIM_Cubic, leaveTangent=0.1355, time=1, value=1}
          ]}}}
       Scale Curve=&FNTimeOfDayScalar(Night Scale=0.4, Morning Scale=1, Evening Scale=0.5)
      )
     )
     Alpha=&ClampFloat(Float=&Add_Float(
       A=$System.Timeline.NormalizedTime
       B=&FloatFromCurve(
        FloatCurve=DataInterface() {propertyValues {Curve {keys [
            {tangentMode=RCTM_User, time=0, value=1}
            {arriveTangent=-0.0405, interpMode=RCIM_Cubic, leaveTangent=-0.0405, tangentMode=RCTM_User, time=1.2, value=1}
            {arriveTangent=-8.0089, interpMode=RCIM_Cubic, leaveTangent=-8.0089, tangentMode=RCTM_User, time=1.5, value=0}
           ]}}}
        CurveIndex=$System.Age
       )
      ))
    )
    Scale RGB=&MakeVectorFromLinearColorRGB(Color=&FNTimeOfDayColor(
      Night Color=<rgba 0.0966 0.0966 0.786 1>
      Morning Color=<rgba 0.0966 0.3264 0.786 1>
      Day Color=<rgba 0.0966 0.3264 0.786 1>
      Evening Color=<rgba 0.0966 0.0966 0.786 1>
     ))
    ScaleA=true
    ScaleRGB=true
   )
   /SetVariables(
    Particles.CoronaColor:LinearColor=&MultiplyLinearColor(
     Linear Color=&FNTimeOfDayColor(Night Color=<rgba 0.7 0.6502 0 1>, Morning Color=<rgba 1.5 1.2603 0 1>, Day Color=<rgba 2 1.3198 0 1>, Evening Color=<rgba 2 1.578 0 1>)
     Scale Factor=&MakeFloat4FromVectorAndFloat(
      "Vector (RGB)"=<2 2 2>
      "Float (Alpha)"=&FN_FrontEndFloatModifier(
       Gameplay=&Lerp_Float(
        A=0.4
        B=1
        Alpha=&ClampFloat(Float=&Add_Float(
          A=$System.Timeline.NormalizedTime
          B=&FloatFromCurve(
           FloatCurve=DataInterface() {propertyValues {Curve {keys [
               {tangentMode=RCTM_User, time=0, value=1}
               {arriveTangent=-0.0405, interpMode=RCIM_Cubic, leaveTangent=-0.0405, tangentMode=RCTM_User, time=1.2, value=1}
               {arriveTangent=-8.0089, interpMode=RCIM_Cubic, leaveTangent=-8.0089, tangentMode=RCTM_User, time=1.5, value=0}
              ]}}}
           CurveIndex=$System.Age
          )
         ))
       )
       Front End=1
      )
     )
    )
    Particles.CoronaPosition:NiagaraPosition=&Add_Vector(
     A=Linked() {linkedVariable {name=Output.SkeletalMeshLocation.SampledPositionWithoutOffset, type=NiagaraPosition}}
     B=&RotateVectorByQuaternion(Vector To Rotate=<0 7 0>, Quaternion=$Particles.SkeletalMeshLocation.SampledRotation)
    )
    Particles.CoronaSize:Vector2f=<60 60>
   ) {isSetParameters=true}
  ]
  rendererClasses [&NiagaraSpriteRendererProperties]
  renderers [{rendererClass=&NiagaraSpriteRendererProperties}]
 }
 FireMesh {
  emitterUpdateScript [
   &EmitterState()
   &SpawnRate(SpawnRate=40, Use Spawn Probability=false)
  ]
  particleSpawnScript [
   &PartitionParticles()
   &InitializeParticle(
    Color=<rgba 1 1 1 0.2>
    Lifetime Max=0.27
    Lifetime Min=0.14
    Lifetime Mode=::NewEnumerator1
    Mesh Scale=<0.1 0.1 0.8>
    Mesh Scale Mode=::NewEnumerator4
    UsePositionOffset=false
    Write Material Random=true
   )
   &InitialMeshOrientation(
    Mesh Axis To Orient=<0 1 0>
    Mesh Orientation Mode=::NewEnumerator2
    Orientation Vector=<0 0 1>
    Rotation=&RandomRangeVector(Minimum=<0 0 -1>, Maximum=<0 0 1>)
    Use Rotation Vector=true
   )
   &CameraOffset()
  ]
  particleUpdateScript [
   &ParticleState()
   &SkeletalMeshLocation(
    Bone / Socket Index=$Particles.PartitionParticles.Partition
    MaskSpawnGroup=false
    Mesh Sampling Type=::NewEnumerator5
    Sampled Position Offset=<0 5 0>
    Skeletal Mesh=$User.SkeletalMesh
    Socket Sampling Mode=::NewEnumerator5
   )
   &ScaleMeshSize(Scale Factor=&MakeVector(
     X=&FloatFromCurve(FloatCurve=DataInterface() {propertyValues {Curve {keys [{tangentMode=RCTM_User, time=0, value=0.75}, {tangentMode=RCTM_User, time=1, value=1.2}]}}})
     Y=&FloatFromCurve(FloatCurve=DataInterface() {propertyValues {Curve {keys [{tangentMode=RCTM_User, time=0, value=0.75}, {tangentMode=RCTM_User, time=1, value=1.2}]}}})
     Z=&FloatFromCurve(
      FloatCurve=DataInterface() {propertyValues {Curve {keys [<at 0 0>,<at 1 1>]}}}
      Scale Curve=&Lerp_Float(
       A=1
       B=0
       Alpha=&ClampFloat(Float=&Add_Float(A=$System.Timeline.NormalizedTime, B=$System.Size scale))
      )
     )
    ))
   &ScaleColor(
    Scale Alpha=&FloatFromCurve(FloatCurve=DataInterface() {propertyValues {Curve {keys [
         {arriveTangent=9.0725, interpMode=RCIM_Cubic, leaveTangent=9.0725, tangentMode=RCTM_User, time=0, value=0}
         {arriveTangent=-1.6443, interpMode=RCIM_Cubic, leaveTangent=-1.6443, tangentMode=RCTM_User, time=0.1408, value=0.9847}
         {arriveTangent=-0.0405, interpMode=RCIM_Cubic, leaveTangent=-0.0405, tangentMode=RCTM_User, time=1, value=0}
        ]}}})
    Scale RGB=&MakeVectorFromLinearColorRGB(Color=&ColorFromCurve(ColorCurve=DataInterface() {propertyValues {
        AlphaCurve {keys [<at 0 1 RCIM_Cubic>]}
        BlueCurve {keys [{arriveTangent=-1.6041, interpMode=RCIM_Cubic, time=0.17, value=0}, <at 0.4736 0>, <at 1 0>]}
        GreenCurve {keys [{arriveTangent=-1.6041, interpMode=RCIM_Cubic, time=0.17, value=1.3991}, <at 0.4736 1.5795>, <at 1 0.8099>]}
        RedCurve {keys [{arriveTangent=-1.6041, interpMode=RCIM_Cubic, time=0.17, value=2}, <at 0.4736 2>, <at 1 1>]}
       }}))
    ScaleA=true
    ScaleRGB=true
   )
  ]
  rendererClasses [&NiagaraMeshRendererProperties]
  renderers [{rendererClass=&NiagaraMeshRendererProperties, rendererIndex=0}]
 }
 FlameMesh {
  emitterUpdateScript [
   &EmitterState()
   &SpawnRate(SpawnRate=20, Use Spawn Probability=false)
  ]
  particleSpawnScript [
   &PartitionParticles()
   &InitializeParticle(Color=<rgba 1 0.8586 0 1>, Mesh Scale=<0.13 0.13 0.4>, Mesh Scale Mode=::NewEnumerator4, UsePositionOffset=false, Write Material Random=true)
   &DynamicMaterialParameters(
    Index 0 Param 1=&RandomRangeFloat(Minimum=0, Maximum=1)
    Index 0 Param 4=&Lerp_Float(A=0, B=20, Alpha=0.5)
    Module.Param0WriteEnabled=true
    Module.Param1WriteEnabled=false
    Module.Param2WriteEnabled=false
    Module.Param3WriteEnabled=true
   )
   &InitialMeshOrientation(
    Mesh Axis To Orient=<0 1 0>
    Mesh Orientation Mode=::NewEnumerator2
    Orientation Vector=<0 0 -1>
    Rotation=&RandomRangeVector(Minimum=<0 0 0>, Maximum=<0 0 1>)
    Use Rotation Vector=true
   )
  ]
  particleUpdateScript [
   &ParticleState()
   &SkeletalMeshLocation(
    Bone / Socket Index=$Particles.PartitionParticles.Partition
    MaskSpawnGroup=false
    Mesh Sampling Type=::NewEnumerator5
    Sampled Position Offset=<0 2 0>
    Skeletal Mesh=$User.SkeletalMesh
    Socket Sampling Mode=::NewEnumerator5
   )
   &ScaleColor(
    Scale Alpha=0.1
    Scale RGB=&MakeVectorFromLinearColorRGB(Color=&ColorFromCurve(ColorCurve=DataInterface() {propertyValues {
        AlphaCurve {keys [<at 0 1 RCIM_Cubic>]}
        BlueCurve {keys [<at 0.0203 0>, <at 0.1448 0>, {arriveTangent=-15.6046, time=0.2896, value=0}, <at 0.6131 0.3807>]}
        GreenCurve {keys [<at 0.0203 3.9784>, <at 0.1448 0.7845>, {arriveTangent=-5.6465, time=0.2896, value=0.7757}, <at 0.6131 0.8705>]}
        RedCurve {keys [<at 0.0203 5>, <at 0.1448 1>, {arriveTangent=-0.716, time=0.2896, value=1}, <at 0.6131 1>]}
       }}))
    ScaleA=true
    ScaleRGB=true
   )
   &ScaleMeshSize(Scale Factor=&FN_FrontEndVector3Modifier(
     Gameplay=&VectorFromCurve(
      VectorCurve=DataInterface() {propertyValues {XCurve {keys [<at 0 1>,<at 1 1>]}, YCurve {keys [<at 0 1>,<at 1 1>]}, ZCurve {keys [<at 0 1>,<at 1 1>]}}}
      Scale Curve=&Lerp_Vector(
       A=<0 0 0>
       B=<1 1 1>
       Alpha=&ClampFloat(Float=&Add_Float(A=$System.Timeline.NormalizedTime, B=$System.Size scale))
      )
     )
     Front End=<1 1 1>
    ))
  ]
  rendererClasses [&NiagaraMeshRendererProperties]
  renderers [{rendererClass=&NiagaraMeshRendererProperties, rendererIndex=0}]
 }
 SideWings {
  emitterSpawnScript [&FN_RibbonIDOffsetUtility()]
  emitterUpdateScript [
   &EmitterState()
   &FN_SpawnAnimTrail(Number of Anim Trails=2)
  ]
  particleSpawnScript [
   &FN_InitializeAnimTrailBoneIndecies(First Bone or Socket Index=2, Second Bone or Socket Index=3)
   &FN_InitializeAnimTrailBoneIndecies(AnimTrail Selection=1, First Bone or Socket Index=4, Second Bone or Socket Index=5) {module=FN_InitializeAnimTrailBoneIndecies001}
   &FN_InitializeAnimTrail(AnimTrailWidth=0.9, Color=<rgba 1 1 1 0.25>, Lifetime=0.05)
   &AddVelocity(Velocity=&MakeCustomVectorFromBool(Bool=$User.bIsFrontEnd, True Vector=<0 2500 0>, False Vector=<0 2500 0>))
  ]
  particleUpdateScript [
   &SolveForcesAndVelocity(Clamp Velocity=false, Limit Acceleration=false)
   &FN_UpdateAnimTrail(SkeletalMesh=$User.SkeletalMesh)
   /SetVariables(Particles.RibbonU0Override=$Particles.NormalizedAge) {isSetParameters=true}
   &ScaleColor(
    Scale Alpha=&FN_FrontEndFloatModifier(Gameplay=&Lerp_Float(A=0, B=0.8, Alpha=$System.Timeline.NormalizedTime), Front End=1)
    ScaleA=true
    ScaleRGB=true
   )
  ]
  rendererClasses [&NiagaraRibbonRendererProperties]
  renderers [{rendererClass=&NiagaraRibbonRendererProperties}]
 }
 Speedlines001 {
  emitterUpdateScript [
   &EmitterState()
   &SpawnRate(
    SpawnRate=&FN_FrontEndFloatModifier(Gameplay=&Lerp_Float(A=0, B=50, Alpha=$User.ForwardDot), Front End=50)
    Use Spawn Probability=false
   )
  ]
  particleSpawnScript [
   /Niagara/Modules/Spawn/Initialization/InitializeParticle(
    Color=&SelectLinearColorFromArray(
     Array Sampling Mode=::NewEnumerator0
     Direct Array Index=?
     Color Selection Array=DataInterface() {propertyValues {ColorData [<rgba 0.6094 0.6094 0.6094 0.3>, <rgba 1 1 1 0.3>]}}
     Random Seed=0
    )
    Lifetime=&FN_FrontEndFloatModifier(
     Gameplay=&UniformRangedFloat(Minimum=0.2, Maximum=0.35)
     Front End=&UniformRangedFloat(Minimum=0.1, Maximum=0.2)
    )
    Module.Write Color=true
    Module.Write Lifetime=true
    Module.Write Mass=true
    Module.Write Material Random=true
    Module.Write Position=false
    Module.Write Scale=false
    Module.Write SpriteRotation=false
    Module.Write SpriteSize=true
    Position=&SimulationPosition()
    Sprite Size=&UniformRangedVector2D(Minimum=<1 4>, Maximum=<3 6>)
   )
   &ShapeLocation(
    Handle Radius=50
    Large Radius=&RandomRangeFloat(Minimum=100, Maximum=150)
    MaskSpawnGroup=false
    Offset=<0 80 25>
    Offset Mode=::NewEnumerator0
    Override Seed=false
    Rotation Axis=<0 1 0>
    Rotation Mode=::NewEnumerator1
    Shape Origin=$Particles.Position
    Shape Primitive=::NewEnumerator3
    UseOwnerScale=true
   )
   &AddVelocityFromPoint(
    Module.Use Falloff Distance=false
    Origin Offset=&MakeCustomVectorFromBool(Bool=$User.bIsFrontEndPreview, True Vector=<0 -500 0>, False Vector=<0 -500 0>)
    Velocity Strength=600
   )
   &AddVelocity(Velocity=&MakeCustomVectorFromBool(
     Bool=$User.bIsFrontEndPreview
     True Vector=&RandomRangeVector(Minimum=<0 300 0>, Maximum=<0 700 0>)
     False Vector=&RandomRangeVector(Minimum=<0 300 0>, Maximum=<0 700 0>)
    ))
  ]
  particleUpdateScript [
   &ParticleState()
   &SolveForcesAndVelocity(Clamp Velocity=false, Limit Acceleration=false)
   &ScaleSpriteSize(Scale Factor=&Vector2DFromCurve(
     Vector2Curve=DataInterface() {propertyValues {XCurve {keys [<at 0 0.5 RCIM_Cubic>,<at 0.5 1 RCIM_Cubic>]}, YCurve {keys [<at 0 0 RCIM_Cubic>,<at 0.5 1 RCIM_Cubic>]}}}
     ScaleCurve=&Vector2DFromFloat(Value=1)
    ))
   &ScaleSpriteSizeBySpeed(Max Scale Factor=<1 40>, Min Scale Factor=<0.01 0.05>, Velocity Threshold=1)
   &ScaleColor(
    Scale Alpha=&FloatFromCurve(
     FloatCurve=DataInterface() {propertyValues {Curve {keys [
         {arriveTangent=0.1307, interpMode=RCIM_Cubic, leaveTangent=0.1307, tangentMode=RCTM_User, time=0, value=0}
         <at 0.3 1>
         {arriveTangent=-2.221, interpMode=RCIM_Cubic, leaveTangent=-2.221, tangentMode=RCTM_User, time=1, value=0}
        ]}}}
     Scale Curve=&MakeCustomFloatFromBool(
      Bool=$User.bIsFrontEndPreview
      True Float=&MakeCustomFloatFromBool(Bool=$User.bIsFullyDeployed, True Float=1, False Float=0)
      False Float=&Lerp_Float(A=0, B=1, Alpha=$User.ForwardDot)
     )
    )
    ScaleA=true
    ScaleRGB=true
   )
  ]
  rendererClasses [&NiagaraSpriteRendererProperties]
  renderers [{rendererClass=&NiagaraSpriteRendererProperties}]
 }
}
summary {
 systemName=NS_Glider_ArcticIceBlue_Trail
 emitters [
  {emitterName=Speedlines001, rendererClasses [&NiagaraSpriteRendererProperties]}
  {emitterName=DynamicBeam001, rendererClasses [&NiagaraRibbonRendererProperties]}
  {emitterName=DynamicBeam004, rendererClasses [&NiagaraRibbonRendererProperties]}
  {emitterName=SideWings, rendererClasses [&NiagaraRibbonRendererProperties]}
  {emitterName=AirThrusters004, rendererClasses [&NiagaraMeshRendererProperties]}
  {emitterName=FlameMesh, rendererClasses [&NiagaraMeshRendererProperties]}
  {emitterName=FireMesh, rendererClasses [&NiagaraMeshRendererProperties]}
  {emitterName=AirThrusters005, rendererClasses [&NiagaraMeshRendererProperties]}
  {emitterName=FakeLighting, rendererClasses [&NiagaraSpriteRendererProperties]}
 ]
 userVariables [
  User.bIsFrontEnd {type=NiagaraBool}
  User.bIsFrontEndPreview {type=NiagaraBool}
  User.bIsFullyDeployed {type=NiagaraBool}
  User.ForwardDot {type=NiagaraFloat, defaultValue=0, description=""}
  User.RibbonIDOffset {type=NiagaraInt32}
  User.RightDot {type=NiagaraFloat, defaultValue=0, description=""}
  User.SkeletalMesh {
   type=NiagaraDataInterfaceSkeletalMesh
   defaultValue {
    struct=NiagaraExt_VariableValue_DataInterface
    value {dataInterface=&self:NiagaraDataInterfaceSkeletalMesh_1}
   }
  }
 ]
}
