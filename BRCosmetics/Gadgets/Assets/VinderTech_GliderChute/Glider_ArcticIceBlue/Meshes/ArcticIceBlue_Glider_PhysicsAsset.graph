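# Physics asset for the Arctic Ice Blue glider chute cosmetic mesh, providing its collision representation.
# It holds a single body on the Frame bone, shaped as one large rotated capsule offset from the bone origin.
# With no constraints, it defines no articulated ragdoll or jointed simulation between parts.
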PhysicsAsset /BRCosmetics/Gadgets/Assets/VinderTech_GliderChute/Glider_ArcticIceBlue/Meshes/ArcticIceBlue_Glider_PhysicsAsset
bodies {Frame [{shapeType=Capsule, length=157.4451, radius=136.264, center=<-0.276 32.4878 38.443>, rotation=<rot 15.0853 89.7564 -90>}]}
constraints []
